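# Niagara ground ice decal effect for the MotoWrestle lance, with both configured emitters disabled and two others showing no content.
# IceDecal and an offset, oriented IceDecal001 each burst-spawn short-lived mesh and sprite particles, gated by User.UseGroundDecal through spawn probability.
# A custom MO2AA module feeds its AA vector into dynamic material parameters, alongside fixed size-like values and two animated curves.
# An alpha curve fades each decal in quickly and out late.
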NiagaraSystem /MotoWrestleGameplay/Assets/VFX/LanceFX/NS_MotoWrestle_Lance_Decal
dependencies {
 usedDataInterfaces [&NiagaraDataInterfaceCurve]
 usedDynamicInputs [&MakeQuatFromAxisAngle]
 usedModules [&self:Mo2AASystem, &SystemState]
}
emitters {
 IceDecal {
  bEnabled=false
  emitterUpdateScript [
   &EmitterState(Life Cycle Mode=::NewEnumerator1, Loop Behavior=::NewEnumerator1, Loop Duration=1, UseLoopDelay=false)
   &SpawnBurst_Instantaneous(
    Spawn Probability=&MakeCustomFloatFromBool(Bool=$User.UseGroundDecal, True Float=1, False Float=0)
    Use Spawn Probability=true
    UseLoopCountLimit=false
   )
  ]
  particleSpawnScript [&InitializeParticle(
    Color Mode=::NewEnumerator0
    Lifetime=3
    Mesh Scale=<8 8 2>
    Mesh Scale Mode=::NewEnumerator4
    UsePositionOffset=true
    Write Material Random=true
   )]
  particleUpdateScript [
   &ParticleState()
   &self:MO2AA() {module=MO2AA}
   &DynamicMaterialParameters(
    Index 0 Param 1=&MakeFloatFromVector4(Vector 4=$Particles.MO2AA.AA, Channel=::NewEnumerator0)
    Index 0 Param 2=&MakeFloatFromVector4(Vector 4=$Particles.MO2AA.AA, Channel=::NewEnumerator1)
    Index 0 Param 3=&MakeFloatFromVector4(Vector 4=$Particles.MO2AA.AA, Channel=::NewEnumerator2)
    Index 0 Param 4=&MakeFloatFromVector4(Vector 4=$Particles.MO2AA.AA, Channel=::NewEnumerator3)
    Index 1 Param 1=800
    Index 1 Param 2=800
    Index 1 Param 3=200
    Index 2 Param 1=&FloatFromCurve(FloatCurve=DataInterface() {propertyValues {Curve {keys [
         {interpMode=RCIM_Cubic, tangentMode=RCTM_Break, time=0.06, value=0}
         {arriveTangent=-1.2116, interpMode=RCIM_Cubic, tangentMode=RCTM_Break, time=1, value=1}
        ]}}})
    Index 3 Param 1=&FloatFromCurve(FloatCurve=DataInterface() {propertyValues {Curve {keys [
         {interpMode=RCIM_Cubic, tangentMode=RCTM_Break, time=0, value=1}
         {arriveTangent=-1.2116, interpMode=RCIM_Cubic, tangentMode=RCTM_Break, time=1, value=-0.7}
        ]}}})
    Module.Index0Param0Write=true
    Module.Index0Param1Write=true
    Module.Index0Param2Write=true
    Module.Index0Param3Write=true
    Module.Index1Param0Write=true
    Module.Index1Param1Write=true
    Module.Index1Param2Write=true
    Module.Index1Param3Write=true
    Module.Index2Param0Write=true
    Module.Index2Param1Write=true
    Module.Index2Param2Write=true
    Module.Index2Param3Write=true
    Module.Index3Param0Write=true
    Module.Index3Param1Write=true
    Module.Index3Param2Write=true
    Module.Index3Param3Write=true
    Write Parameter Index 1=true
    Write Parameter Index 2=true
    Write Parameter Index 3=true
   )
   &ScaleColor(
    Scale Alpha=&FloatFromCurve(FloatCurve=DataInterface() {propertyValues {Curve {keys [
         {interpMode=RCIM_Cubic, tangentMode=RCTM_Break, time=0, value=0}
         {interpMode=RCIM_Cubic, tangentMode=RCTM_Break, time=0.03, value=1}
         {arriveTangent=0.1495, interpMode=RCIM_Cubic, leaveTangent=0.1495, tangentMode=RCTM_User, time=0.7, value=1}
         <at 1 0>
        ]}}})
    ScaleA=true
    ScaleRGB=true
   )
  ]
  rendererClasses [&NiagaraMeshRendererProperties, &NiagaraSpriteRendererProperties]
  renderers [
   {rendererClass=&NiagaraMeshRendererProperties, rendererIndex=0}
   {rendererClass=&NiagaraSpriteRendererProperties, rendererIndex=1}
  ]
 }
 IceDecal001 {
  bEnabled=false
  emitterUpdateScript [
   &EmitterState(Life Cycle Mode=::NewEnumerator1, Loop Behavior=::NewEnumerator1, Loop Duration=1, UseLoopDelay=false)
   &SpawnBurst_Instantaneous(
    Spawn Probability=&MakeCustomFloatFromBool(Bool=$User.UseGroundDecal, True Float=1, False Float=0)
    Use Spawn Probability=true
    UseLoopCountLimit=false
   )
  ]
  particleSpawnScript [
   &InitializeParticle(
    Color Mode=::NewEnumerator0
    Lifetime=3
    Mesh Scale=<8 8 3>
    Mesh Scale Mode=::NewEnumerator4
    Position Offset=<0 100 0>
    UsePositionOffset=true
    Write Material Random=true
   )
   &InitialMeshOrientation(Mesh Orientation Mode=::NewEnumerator2, Orientation Vector=<-1 0 0>, Use Rotation Vector=false)
  ]
  particleUpdateScript [
   &ParticleState()
   &self:MO2AA() {module=MO2AA}
   &DynamicMaterialParameters(
    Index 0 Param 1=&MakeFloatFromVector4(Vector 4=$Particles.MO2AA.AA, Channel=::NewEnumerator0)
    Index 0 Param 2=&MakeFloatFromVector4(Vector 4=$Particles.MO2AA.AA, Channel=::NewEnumerator1)
    Index 0 Param 3=&MakeFloatFromVector4(Vector 4=$Particles.MO2AA.AA, Channel=::NewEnumerator2)
    Index 0 Param 4=&MakeFloatFromVector4(Vector 4=$Particles.MO2AA.AA, Channel=::NewEnumerator3)
    Index 1 Param 1=250
    Index 1 Param 2=600
    Index 1 Param 3=300
    Index 2 Param 1=&FloatFromCurve(FloatCurve=DataInterface() {propertyValues {Curve {keys [
         {interpMode=RCIM_Cubic, tangentMode=RCTM_Break, time=0.06, value=0}
         {arriveTangent=-1.2116, interpMode=RCIM_Cubic, tangentMode=RCTM_Break, time=1, value=1}
        ]}}})
    Index 3 Param 1=&FloatFromCurve(FloatCurve=DataInterface() {propertyValues {Curve {keys [
         {interpMode=RCIM_Cubic, tangentMode=RCTM_Break, time=0, value=1}
         {arriveTangent=-1.2116, interpMode=RCIM_Cubic, tangentMode=RCTM_Break, time=1, value=-0.7}
        ]}}})
    Module.Index0Param0Write=true
    Module.Index0Param1Write=true
    Module.Index0Param2Write=true
    Module.Index0Param3Write=true
    Module.Index1Param0Write=true
    Module.Index1Param1Write=true
    Module.Index1Param2Write=true
    Module.Index1Param3Write=true
    Module.Index2Param0Write=true
    Module.Index2Param1Write=true
    Module.Index2Param2Write=true
    Module.Index2Param3Write=true
    Module.Index3Param0Write=true
    Module.Index3Param1Write=true
    Module.Index3Param2Write=true
    Module.Index3Param3Write=true
    Write Parameter Index 1=true
    Write Parameter Index 2=true
    Write Parameter Index 3=true
   )
   &ScaleColor(
    Scale Alpha=&FloatFromCurve(FloatCurve=DataInterface() {propertyValues {Curve {keys [
         {interpMode=RCIM_Cubic, tangentMode=RCTM_Break, time=0, value=0}
         {interpMode=RCIM_Cubic, tangentMode=RCTM_Break, time=0.03, value=1}
         {arriveTangent=0.1495, interpMode=RCIM_Cubic, leaveTangent=0.1495, tangentMode=RCTM_User, time=0.7, value=1}
         <at 1 0>
        ]}}})
    ScaleA=true
    ScaleRGB=true
   )
  ]
  rendererClasses [&NiagaraMeshRendererProperties, &NiagaraSpriteRendererProperties]
  renderers [
   {rendererClass=&NiagaraMeshRendererProperties, rendererIndex=0}
   {rendererClass=&NiagaraSpriteRendererProperties, rendererIndex=1}
  ]
 }
 IceDecal002 {rendererClasses [], renderers []}
 IceDecal003 {rendererClasses [], renderers []}
}
summary {
 systemName=NS_MotoWrestle_Lance_Decal
 emitters [
  {
   bEnabled=false
   emitterName=IceDecal
   rendererClasses [&NiagaraMeshRendererProperties, &NiagaraSpriteRendererProperties]
  }
  {
   bEnabled=false
   emitterName=IceDecal001
   rendererClasses [&NiagaraMeshRendererProperties, &NiagaraSpriteRendererProperties]
  }
  {emitterName=IceDecal002, rendererClasses []}
  {emitterName=IceDecal003, rendererClasses []}
 ]
 userVariables [User.UseGroundDecal {type=NiagaraBool, defaultValue=true}]
}
