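# Niagara system for the lance target impact effect in MotoWrestle, with user parameters for color variant, scale, ground decal, water and Ultra quality.
# System spawn modules include a line trace for relocation.
# Spark001 bursts bluish GPU sprites with vertically flattened velocity, drag, speed-based stretching and a late shrink.
# Glow, spike, Fresnel sphere and ice mesh emitters are listed without renderers or module detail.
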NiagaraSystem /MotoWrestleGameplay/Assets/VFX/LanceFX/NS_MotoWrestle_Lance_Target_Impact
dependencies {
 usedDataInterfaces [&NiagaraDataInterfaceCurve, &NiagaraDataInterfaceCollisionQuery]
 usedDynamicInputs [
  &UniformRangedFloat
  &MakeIntFromFloat
  &MakeCustomFloatFromBool
  &TransformPosition
  &AddVectorToPosition
  &Multiply_Float
  &FN_LOD_Distance_Remap
 ]
 usedModules [
  &self:SystemSpawnScript.SystemScriptSource.SystemScriptGraph.NiagaraNodeAssignment_0.SetVariables_#1
  &self:LineTraceForRelocation
  &self:SystemSpawnScript.SystemScriptSource.SystemScriptGraph.NiagaraNodeAssignment_1.SetVariables_#2
  &self:SystemSpawnScript.SystemScriptSource.SystemScriptGraph.NiagaraNodeAssignment_2.SetVariables_#3
  &self:SystemSpawnScript.SystemScriptSource.SystemScriptGraph.NiagaraNodeAssignment_3.SetVariables_#4
  &self:SystemSpawnScript.SystemScriptSource.SystemScriptGraph.NiagaraNodeAssignment_4.SetVariables_#5
  &SystemState
 ]
}
emitters {
 BigGlow {rendererClasses [], renderers []}
 BigGlow2 {rendererClasses [], renderers []}
 Blast_Spike {rendererClasses [], renderers []}
 FresnelSphere {rendererClasses [], renderers []}
 IceMesh {rendererClasses [], renderers []}
 IceMeshParticle {rendererClasses [], renderers []}
 Spark001 {
  simTarget=GPUComputeSim
  emitterUpdateScript [
   &EmitterState()
   &SpawnBurst_Instantaneous(Spawn Count=60, Use Spawn Probability=false, UseLoopCountLimit=false)
  ]
  particleSpawnScript [
   &InitializeParticle(
    Color=<rgba 1 1.3333 2 1>
    Lifetime=&RandomRangeFloat(Minimum=0.2, Maximum=0.3)
    Mesh Scale Mode=::NewEnumerator4
    Sprite Size=&RandomRangeVector2D(Minimum=<5 25>, Maximum=<15 80>)
    Sprite Size Mode=::NewEnumerator4
    UsePositionOffset=false
    Write Material Random=true
   )
   &ShapeLocation(MaskSpawnGroup=false, Override Seed=false, UseOwnerScale=true)
   &AddVelocity(
    Module.Use Falloff Distance=false
    Override Seed=false
    Velocity Mode=::NewEnumerator1
    Velocity Speed=&RandomRangeFloat(Minimum=664.1049, Maximum=3263.1738)
   )
   /SetVariables(Particles.Velocity=&Multiply_Vector(A=$Particles.Velocity, B=<1 1 0.7>)) {isSetParameters=true}
   &DynamicMaterialParameters(
    Index 0 Param 2=1
    Index 0 Param 3=1.3
    Index 0 Param 4=0.04
    Index 1 Param 2=0.1
    Index 1 Param 3=0.5
    Module.Index0Float0Write=false
    Module.Index0Float1Write=true
    Module.Index0Float2Write=true
    Module.Index0Float3Write=true
    Module.Index1Float0Write=false
    Module.Index1Float1Write=true
    Module.Index1Float2Write=true
    Module.Index1Float3Write=false
    Write Parameter Index 1=::NewEnumerator1
   )
  ]
  particleUpdateScript [
   &Drag(Drag=&RandomRangeFloat(Minimum=4.0604, Maximum=8.0179), Use Linear Drag=true, Use Rotational Drag=true)
   &SolveForcesAndVelocity(Clamp Velocity=false, Limit Acceleration=false)
   &ParticleState()
   &ScaleSpriteSizeBySpeed(Max Scale Factor=<1 2>, Min Scale Factor=<1 0.25>, Velocity Threshold=281.7375)
   &ScaleSpriteSize(
    Scale Factor=&MakeVector2D(
     X=1
     Y=&FloatFromCurve(FloatCurve=DataInterface() {propertyValues {Curve {keys [<at 0.8 1>,<at 1 0>]}}})
    )
    Scale Sprite Size Mode=::NewEnumerator2
   )
   &FN_Sprite_Attributes(
    Color=&MultiplyLinearColor(Linear Color=$Particles.Color, Scale Factor=<1 1 1 0.05>)
    Create Color Attribute=true
    Create SpriteSize Attribute=true
    SpriteSize=&Multiply_Vector2D(A=$Particles.SpriteSize, B=<15 2>)
    Write to Module Namespace=true
   )
  ]
  rendererClasses [&NiagaraSpriteRendererProperties]
  renderers [{rendererClass=&NiagaraSpriteRendererProperties}]
 }
}
summary {
 systemName=NS_MotoWrestle_Lance_Target_Impact
 emitters [
  {emitterName=Blast_Spike, rendererClasses []}
  {emitterName=IceMesh, rendererClasses []}
  {emitterName=FresnelSphere, rendererClasses []}
  {emitterName=IceMeshParticle, rendererClasses []}
  {emitterName=BigGlow, rendererClasses []}
  {emitterName=BigGlow2, rendererClasses []}
  {emitterName=Spark001, simTarget=GPUComputeSim, rendererClasses [&NiagaraSpriteRendererProperties]}
 ]
 userVariables [
  User.ColorVariant {type=NiagaraInt32}
  User.FX_DecalScale {type=NiagaraFloat, defaultValue=0.75, description=""}
  User.Global Scale {type=NiagaraFloat, defaultValue=1, description=""}
  User.IsUltra {type=NiagaraBool, defaultValue=true}
  User.UseGroundDecal {type=NiagaraBool, defaultValue=true}
  User.WaterFX {type=NiagaraBool}
 ]
}
